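; Gameplay ability that plays NPC speech-bubble barks during base defense and reclamation activities, ending immediately if the pawn starts outside the safe zone.
; Async gameplay-event waits trigger rescue, active and completion barks, with completion ending the ability.
; Repeating timers, each interval randomized when set, loop active barks and swap to downed barks during DBNO, resuming on revive.
; Damage barks are throttled, and ending the ability clears timers and unbinds pawn events.
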(event Ability|EventActivateAbility
  (bind _returnvalue (Ability|GetActivatingPawn))
  ;[0 0 0] Activate Ability
  (|CallParentFunction)
  (Utilities|IsValid _returnvalue
    (:Is%20Valid
      (Variables|OwnerInfo|SetOwningPawn _returnvalue)
      (if (CallFunction|IsOwnerinSafeZone)
        (CallFunction|StartListeningtoBaseDefense)
        (CallFunction|StartListeningtoBaseReclamation)
        (CallFunction|StartListeningtoDamageandDBNO)
        (else
          (Ability|EndAbility))))
    (:Is%20Not%20Valid
      (Ability|EndAbility))))

(event Custom|PlayRescueBarks
  ;[0 0 0] Base Reclamation
  ;;[0.1144 0.1144 0.1144] NPC freed from the sac
  (Variables|Default|SetbIsBaseDefense false)
  (CallFunction|PlayRandomBarkforTag (Variables|Barks|GetReclaimBarkRescueTag) true)
  (Utilities|FlowControl|Delay 6.0)
  (CallFunction|PlayRandomBarkforTag (Variables|Barks|GetReclaimBarkRescueTagSecond))
  (CallFunction|PlayReclaimActiveBarks))

(event Custom|PlayDefenseActiveBarks
  ;[0 0 0] Base Defense
  ;;[0.1144 0.1144 0.1144] Base Active
  (Variables|Default|SetbIsBaseDefense true)
  (if (not (Variables|Default|GetbBaseComplete))
    (bind _returnvalue (Utilities|Time|SetTimerbyEvent (delegate Custom|DefenseActiveLoop) (Math|Random|RandomFloatinRange (Variables|Default|GetMinTimeToRandomBark) (Variables|Default|GetMaxTimeToRandomBark)) true))
    (Variables|TimerHandles|SetDefenseActiveTimerHandle _returnvalue)))

(event Custom|DefenseActiveLoop
  ;[0 0 0] Base Defense
  ;;[0.1144 0.1144 0.1144] Base Active
  (CallFunction|PlayRandomBarkforTag (Variables|Barks|GetDefenseBaseActiveBarkTag) true))

(event Custom|PlayReclaimActiveBarks
  ;[0 0 0] Base Reclamation
  ;;[0.1144 0.1144 0.1144] NPC Freed, Reclamation Active
  (if (not (Variables|Default|GetbBaseComplete))
    (bind _returnvalue (Utilities|Time|SetTimerbyEvent (delegate Custom|ReclaimActiveLoop) (Math|Random|RandomFloatinRange (Variables|Default|GetMinTimeToRandomBark) (Variables|Default|GetMaxTimeToRandomBark)) true))
    (Variables|TimerHandles|SetReclaimActiveTimerHandle _returnvalue)))

(event Custom|ReclaimActiveLoop
  ;[0 0 0] Base Reclamation
  ;;[0.1144 0.1144 0.1144] NPC Freed, Reclamation Active
  (CallFunction|PlayRandomBarkforTag (Variables|Barks|GetReclaimBarkActiveTag) true))

(event Custom|PlayDefenseCompleteBarks
  ;[0 0 0] Base Defense
  ;;[0.1144 0.1144 0.1144] Base Complete
  (Variables|Default|SetbBaseComplete true)
  (if (CallFunction|IsOwnerinSafeZone)
    (CallFunction|ClearActiveTimerifValid (Variables|TimerHandles|GetDefenseActiveTimerHandle))
    (Utilities|FlowControl|Delay 5.0)
    (CallFunction|PlayRandomBarkforTag (Variables|Barks|GetDefenseBaseCompleteBarkTag) true)
    (Utilities|FlowControl|Delay 1e+01)
    (else
      _))
  (Ability|EndAbility))

(event Custom|PlayReclaimCompleteBarks
  ;[0 0 0] Base Reclamation
  ;;[0.1144 0.1144 0.1144] Reclamation Complete
  (Variables|Default|SetbBaseComplete true)
  (if (CallFunction|IsOwnerinSafeZone)
    (CallFunction|ClearActiveTimerifValid (Variables|TimerHandles|GetReclaimActiveTimerHandle))
    (Utilities|FlowControl|Delay 5.0)
    (CallFunction|PlayRandomBarkforTag (Variables|Barks|GetReclaimBarkCompleteTag) true)
    (Utilities|FlowControl|Delay 1e+01)
    (else
      _))
  (Ability|EndAbility))

(event Custom|ClearActiveTimerIfValid
  (params (TimerHandle :type Timer%20Handle%20Structure))
  ;[0 0 0] Helper events
  (if (Utilities|Time|IsValidTimerHandle TimerHandle)
    (Utilities|Time|ClearandInvalidateTimerbyHandle TimerHandle)))

(event Custom|PlayRandomBarkForTag
  (params (BarkTag :type Gameplay%20Tag%20Structure) (bDelay :type Boolean))
  ;[0 0 0] Helper events
  (if bDelay
    (Utilities|FlowControl|Delay (Math|Random|RandomFloatinRange 0.5 2.5))
    (else
      _))
  (Class|BPFLAICommonFunctions|PlaySpeechBubble (Variables|OwnerInfo|GetOwningPawn) BarkTag))

(event Custom|StartListeningToBaseDefense
  ;[0 0 0] Helper events
  ;;[0.1144 0.1144 0.1144] Wait for the different states of the base defense event to occur
  (Ability|Async|WaitGameplayEventToActor (Variables|OwnerInfo|GetOwningPawn) (Variables|EventTags|GetDefenseEventActive) true
    (:then
      (Ability|Async|WaitGameplayEventToActor (Variables|OwnerInfo|GetOwningPawn) (Variables|EventTags|GetDefenseEventComplete) true
        (:EventReceived
          (CallFunction|PlayDefenseCompleteBarks))))
    (:EventReceived
      (CallFunction|PlayDefenseActiveBarks))))

(event Custom|StartListeningToBaseReclamation
  ;[0 0 0] Helper events
  ;;[0.1144 0.1144 0.1144] Wait for the different states of the base reclamation event to occur
  (Ability|Async|WaitGameplayEventToActor (Variables|OwnerInfo|GetOwningPawn) (Variables|EventTags|GetReclaimEventRescue) true
    (:then
      (Ability|Async|WaitGameplayEventToActor (Variables|OwnerInfo|GetOwningPawn) (Variables|EventTags|GetReclaimEventActive) true
        (:then
          (Ability|Async|WaitGameplayEventToActor (Variables|OwnerInfo|GetOwningPawn) (Variables|EventTags|GetReclaimEventComplete) true
            (:EventReceived
              (CallFunction|PlayReclaimCompleteBarks))))
        (:EventReceived
          (CallFunction|PlayReclaimActiveBarks))))
    (:EventReceived
      (CallFunction|PlayRescueBarks))))

(event Custom|StartListeningToDamageAndDBNO
  ;[0 0 0] Helper events
  ;;[0.15 0.15 0.15] Wait for damage or DBNO on the NPC
  (EventDispatchers|BindEventtoOnDamaged (Variables|OwnerInfo|GetOwningPawn) (EventDispatchers|CreateEvent :!event Custom|OnPawnDamaged))
  (EventDispatchers|BindEventtoOnDBNOStateChanged (Variables|OwnerInfo|GetOwningPawn) (EventDispatchers|CreateEvent :!event Custom|DBNOStateChanged)))

(event Ability|EventFailedToActivatePassiveAbility
  (params (ActorInfo :type Gameplay%20Ability%20Actor%20Info%20Structure))
  ;[0 0 0] Helper events
  (|CallParentFunction ActorInfo))

(event Custom|OnPawnDamaged
  (params (DamagedActor :type Actor%20Object%20Reference) (Damage :type Float%20%28single-precision%29) (InstigatedBy :type Controller%20Object%20Reference) (DamageCauser :type Actor%20Object%20Reference) (HitLocation :type Vector) (FHitComponent :type Primitive%20Component%20Object%20Reference) (BoneName :type Name) (Momentum :type Vector))
  ;[0 0 0] NPC damaged
  (if (Variables|Default|GetbBaseComplete)
    _
    (elif (and (> Damage 2.5) (not (Variables|OwnerInfo|GetbRecentlyDamaged)))
      (CallFunction|PlayRandomBarkforTag (Variables|Barks|GetPawnDamagedBarkTag))
      (Variables|OwnerInfo|SetbRecentlyDamaged true)
      (Utilities|FlowControl|Delay (Variables|Default|GetMinTimeToRandomBark))
      (Variables|OwnerInfo|SetbRecentlyDamaged false))))

(event Custom|DBNOStateChanged
  (params (FortPawn :type Fort%20Pawn%20Object%20Reference) (bInIsDBNO :type Boolean))
  ;[0 0 0] NPC needs revive or is revived
  (if bInIsDBNO
    (CallFunction|PlayRandomBarkforTag (Variables|Barks|GetPawnDBNOBarkTag))
    (bind _returnvalue (Utilities|Time|SetTimerbyEvent (delegate Custom|BarkLoopDBNO) (Math|Random|RandomFloatinRange 1e+01 2e+01) true))
    (Variables|TimerHandles|SetDBNO_TimerHandle _returnvalue)
    (if (Variables|Default|GetbIsBaseDefense)
      (CallFunction|ClearActiveTimerifValid (Variables|TimerHandles|GetDefenseActiveTimerHandle))
      (else
        (CallFunction|ClearActiveTimerifValid (Variables|TimerHandles|GetReclaimActiveTimerHandle))))
    (else
      (CallFunction|ClearActiveTimerifValid (Variables|TimerHandles|GetDBNO_TimerHandle))
      (if (not (Variables|Default|GetbBaseComplete))
        (CallFunction|PlayRandomBarkforTag (Variables|Barks|GetOnReviveBark))
        (if (Variables|Default|GetbIsBaseDefense)
          (CallFunction|PlayDefenseActiveBarks)
          (else
            (CallFunction|PlayReclaimActiveBarks)))))))

(event Custom|BarkLoopDBNO
  ;[0 0 0] NPC needs revive or is revived
  (CallFunction|PlayRandomBarkforTag (Variables|Barks|GetPawnDBNOBarkTag) true))

(event Ability|EventOnEndAbility
  (params (bWasCancelled :type Boolean))
  ;[0 0 0] On end ability - clear any active timers and unbind events
  (|CallParentFunction bWasCancelled)
  (CallFunction|ClearActiveTimerifValid (Variables|TimerHandles|GetDefenseActiveTimerHandle))
  (CallFunction|ClearActiveTimerifValid (Variables|TimerHandles|GetDBNO_TimerHandle))
  (CallFunction|ClearActiveTimerifValid (Variables|TimerHandles|GetReclaimActiveTimerHandle))
  (EventDispatchers|UnbindEventfromOnDamaged (Variables|OwnerInfo|GetOwningPawn) (EventDispatchers|CreateEvent :!event Custom|OnPawnDamaged))
  (EventDispatchers|UnbindEventfromOnDBNOStateChanged (Variables|OwnerInfo|GetOwningPawn) (EventDispatchers|CreateEvent :!event Custom|DBNOStateChanged)))

(fn IsOwnerInSafeZone
  (bind _asfort_player_pawn (Utilities|Casting|CastToFortPlayerPawn (Variables|OwnerInfo|GetOwningPawn)))
  (return (Variables|/Script/FortniteGame.FortPlayerPawn|GetbIsInsideSafeZone _asfort_player_pawn)))
